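; Creative interaction behavior for possessing props, which applies a minigame's Prop Hunt settings to what a player may possess.
; CanInteract refuses pawns with the DisablePrimaryFire tag and, during an active minigame that disallows changing props, pawns with an active possession ability.
; IsObjectClassForbidden extends the parent's check with the settings' disallowed prop list, matching by exact class.
; Both read those settings through the PropPossessorConfig interface.
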(fn CanInteract
  (bind _pawn (CallFunction|GetFortPlayerPawn))
  (if (GameplayTags|HasMatchingGameplayTag _pawn "(TagName=\"Gameplay.Status.DisablePrimaryFire\")")
    (return false)
    (else
      (bind _minigame (Utilities|GetMinigame _pawn)
        (:Valid
          (Minigame|IsInProgress _minigame
            (:Yes
              (bind _asprop_possessor_config_interface (Utilities|Casting|CastToPropPossessorConfigInterface (Variables|/Game/Athena/Playset/Minigames/Minigame_PlayerCreated.Minigame_PlayerCreated_C|GetPropHuntSettings _minigame))
                (:then
                  (bind (_show_props_remaining _prop_health_scaling _allow_change_of_prop) (CreativePropPossession|GetPropDetails _asprop_possessor_config_interface))
                  (if (not _allow_change_of_prop)
                    (return (not (CallFunction|IsPossessingProp)))))
                (:CastFailed
                  _))
              (bind _returnvalue (|CallParentFunction))
              (return _returnvalue))
            (:No
              _)))
        (:Invalid
          _)))))

(fn GetFortPlayerPawn
  (bind _output_get (LocalVariables|SetFortPC (CreativeInteraction|GetFortPlayerController)))
  (Utilities|IsValid _output_get
    (:Is%20Valid
      (return (Game|Player|GetPlayerPawn _output_get)))
    (:Is%20Not%20Valid
      (return 0))))

(fn IsObjectClassForbidden
  (params (Class :type Object%20Class%20Reference))
  ;[1 1 1] If base logic determines the prop is invalid, exit early with the same value.
  ; Note: Potentially we may want to add an option to completely override the base logic here, but for now we're essentially "appending" to the pre-existing disallow list
  (bind _returnvalue (|CallParentFunction Class))
  (if _returnvalue
    (return true)
    (else
      (bind _minigame (Utilities|GetMinigame (CallFunction|GetFortPlayerPawn))
        (:Valid
          (bind _asprop_possessor_config_interface (Utilities|Casting|CastToPropPossessorConfigInterface (Variables|/Game/Athena/Playset/Minigames/Minigame_PlayerCreated.Minigame_PlayerCreated_C|GetPropHuntSettings _minigame)))
          (bind _disallowed_prop_list (CreativePropPossession|GetFilteringDetails _asprop_possessor_config_interface))
          ;[1 1 1] Loop through all the specified Actor References and determine if the prop we're looking to possess is disallowed
          (for _array_element _disallowed_prop_list
            (if (== (Utilities|GetClass _array_element) Class)
              (return true))))))))

(fn IsPossessingProp
  (bind _pawn (CallFunction|GetFortPlayerPawn))
  (Utilities|IsValid _pawn
    (:Is%20Valid
      (bind _asfort_ability_system_component_athena (Utilities|Casting|CastToFortAbilitySystemComponentAthena (Ability|GetAbilitySystemComponent _pawn)))
      (return (Abilities|HasActiveAbilitywithAnyMatchingTag _asfort_ability_system_component_athena "(GameplayTags=((TagName=\"Creative.PropPossessor.Active.Ability\")))")))))
