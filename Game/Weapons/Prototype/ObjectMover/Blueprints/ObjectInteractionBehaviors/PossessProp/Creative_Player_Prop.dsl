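; Creative prop-disguise actor that lets a player possess a copied prop, as used in Prop Hunt.
; It attaches the prop to the pawn, timer-lerps pawn scale and capsule to prop bounds, and runs material transition VFX, with RepNotifies driving client setup.
; Prop Hunt settings set the hint-sound interval, synced to server time across props with a HUD countdown.
; Ending possession restores capsule, scale and weapon visibility, then removes prop and actor.
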(event Custom|SetupTimer
  (params (Time%20Remaining :type Float%20%28double-precision%29))
  (Networking|CanExecuteCosmeticEvents
    (:False
      (Variables|Default|SetSoundTimeRemaining Time%20Remaining)
      (bind _returnvalue (Utilities|Time|SetTimerbyEvent (delegate Custom|UI%20Countdown%20Timer%20Event) 1.0 true))
      (Variables|Default|SetUICountdownTimerHandle _returnvalue))))

(event Custom|UI%20Countdown%20Timer%20Event
  (bind _soundtimeremaining (Variables|Default|GetSoundTimeRemaining))
  (bind _returnvalue (- _soundtimeremaining 1.0))
  ;[1 1 1] UI Countdown Notification
  (Variables|Default|SetSoundTimeRemaining _returnvalue)
  (if (< _returnvalue 0.0)
    (Utilities|Time|ClearandInvalidateTimerbyHandle (Variables|Default|GetUICountdownTimerHandle)) ; Clear any previous timers
  ))

(event Custom|SyncToHintTimer
  (bind _returnvalue (Fortnite|Sync|Time|GetServerWorldTimeSeconds))
  (bind _currentlooptimeforhintsound (Variables|HintSound|GetCurrentLoopTimeForHintSound))
  (bind _returnvalue_2 (Math|Float|%25%28Float%29 _returnvalue 6e+01))
  (bind _returnvalue_3 (Math|Float|%25%28Float%29 _returnvalue_2 _currentlooptimeforhintsound))
  (bind _pawnattachedto (Variables|Default|GetPawnAttachedTo))
  (bind _playerstate (Variables|/Script/Engine.Pawn|GetPlayerState _pawnattachedto))
  (bind _returnvalue_4 (Minigame|GetMinigame _playerstate :!class /Script/FortniteGame.MinigameBlueprintLibrary))
  (bind _returnvalue_5 (- _currentlooptimeforhintsound _returnvalue_3))
  ;[1 1 1] Gets the prop onto the correct interval to match the other props.
  (if (Networking|IsServer)
    (if (> (Variables|HintSound|GetCurrentLoopTimeForHintSound) 0.0)
      (Utilities|IsValid _returnvalue_4
        (:Is%20Valid
          (Minigame|IsInProgress _returnvalue_4
            (:Yes
              (Utilities|Time|SetTimerbyEvent (delegate Custom|StartHidingHintSound) _returnvalue_5) ; Wait the remaining seconds to get on the correct interval.
              (CallFunction|SetupTimer (Math|Conversions|ToFloat(Integer) (Math|Float|Round _returnvalue_5))))))))))

(event Custom|StartHidingHintSound
  ;[1 1 1] All props play a sound at the same interval
  (if (Networking|IsServer)
    (Utilities|Time|ClearandInvalidateTimerbyHandle (Variables|Default|GetUICountdownTimerHandle)) ; Clear any previous timers
    (bind _returnvalue (Utilities|Time|SetTimerbyEvent (delegate Custom|ServerTimerReset) (Variables|HintSound|GetCurrentLoopTimeForHintSound) true)) ; wait the interval then play again
    (Variables|Default|SetHintSoundTimer _returnvalue)
    (CallFunction|SetupTimer (Variables|HintSound|GetCurrentLoopTimeForHintSound))
    (else
      (CallFunction|PlayHintSound) ; Play sound right away to get on cadence
      (Utilities|Time|ClearandInvalidateTimerbyHandle (Variables|Default|GetHintSoundTimer)) ; Clear any previous timers
      (bind _returnvalue_2 (Utilities|Time|SetTimerbyEvent (delegate Custom|PlayHintSound) (Variables|HintSound|GetCurrentLoopTimeForHintSound) true)) ; wait the interval then play again
      (Variables|Default|SetHintSoundTimer _returnvalue_2))))

(event Custom|PlayHintSound
  (bind _pawnattachedto (Variables|Default|GetPawnAttachedTo))
  (bind _returnvalue (Transformation|GetActorLocation _pawnattachedto))
  ;[1 1 1] All props play a sound at the same interval
  (Audio|PlaySoundatLocation (Variables|Audio|GetPropHintSoundFX) _returnvalue)
  (Fortnite|ShowSoundIndicatorLocalforTeam :Location _returnvalue :MaxAudibleDistance (* (Variables|GridSnapping|GetSnapGridSize) 3.0) :IndicatorType "FootStep" :Team (Team|GetActorTeam self) :Affiliations (Utilities|Array|MakeArray "Hostile") :Tint "(R=0.000000,G=0.000000,B=0.000000,A=0.000000)" :OverrideIcon "/Game/UI/Foundation/Textures/Icons/Audio/T-Icon-Sound-On-M.T-Icon-Sound-On-M" :AttenuationMaxRangeMultiplier 1.0 :AttenuationIntensityMultiplier 1.0))

(event Custom|OnDamaged_Event_0
  (params (DamagedActor :type Actor%20Object%20Reference) (Damage :type Float%20%28single-precision%29) (InstigatedBy :type Controller%20Object%20Reference) (DamageCauser :type Actor%20Object%20Reference) (HitLocation :type Vector) (FHitComponent :type Primitive%20Component%20Object%20Reference) (BoneName :type Name) (Momentum :type Vector))
  ;[1 0.09063 0.09063] NOT USED -- Checks damage taken from health drain
  (if (Math|Float|NearlyEqual(Float) Damage (Variables|Default|GetDrainDamageIncrement))
    (bind _output_get (Variables|Default|SetDrainDamageCount (+ (Variables|Default|GetDrainDamageCount) Damage)))
    (bind _asb_creative_prop_selector_athena (Utilities|Casting|CastToB_CreativePropSelector_Athena (Variables|/Script/FortniteGame.FortPawn|GetCurrentWeapon (Variables|Default|GetPawnAttachedTo))))
    (Variables|/Game/Weapons/Prototype/ObjectMover/Blueprints/B_CreativePropSelector_Athena.B_CreativePropSelector_Athena_C|SetPossessedPropHealDrainAmount (- 1.0 (/ _output_get (Variables|Health|GetCurrentMaxHealth))) _asb_creative_prop_selector_athena)))

(event Custom|ServerTimerReset
  ;[1 1 1] All props play a sound at the same interval
  (Utilities|Time|ClearandInvalidateTimerbyHandle (Variables|Default|GetUICountdownTimerHandle)) ; Clear any previous timers
  (CallFunction|SetupTimer (Variables|HintSound|GetCurrentLoopTimeForHintSound)))

(event EventBeginPlay
  (bind _returnvalue (Game|GetInstigatorController))
  (|CallParentFunction)
  (bind _asplayer_pawn_athena (Utilities|Casting|CastToPlayerPawn_Athena (Game|GetInstigator)))
  (Variables|Default|SetPawnAttachedTo _asplayer_pawn_athena)
  (CallFunction|DoPawnRelatedInitialization)
  (bind (_settings _exists) (CallFunction|GetPropHuntSettings))
  (if _exists
    (bind (_show_ping_timer _ping_frequency) (CreativePropPossession|GetPingDetails _settings))
    (Variables|HintSound|SetCurrentLoopTimeForHintSound _ping_frequency)
    (bind _asminigame_player_created (Utilities|Casting|CastToMinigame_PlayerCreated (Minigame|GetMinigame (Variables|/Script/Engine.Pawn|GetPlayerState (Variables|Default|GetPawnAttachedTo)) :!class /Script/FortniteGame.MinigameBlueprintLibrary)))
    (Default|BindEventtoOnPropHuntSettingsChanged _asminigame_player_created (delegate Custom|OnPropHuntSettingsChanged))
    (if (Networking|IsServer)
      (Utilities|IsValid _returnvalue
        (:Is%20Valid
          (CreativePropPossession|OnPlayerBeginEnteringDisguise(Message) (Utilities|ToObject(Interface) _settings) _returnvalue))))))

(event Custom|OnPropHuntSettingsChanged
  (params (Minigame :type Minigame%20Player%20Created%20Object%20Reference))
  (bind (_settings _exists) (CallFunction|GetPropHuntSettings))
  (if _exists
    (bind (_show_ping_timer _ping_frequency) (CreativePropPossession|GetPingDetails _settings))
    (if (!= _ping_frequency (Variables|HintSound|GetCurrentLoopTimeForHintSound))
      (Variables|HintSound|SetCurrentLoopTimeForHintSound _ping_frequency)
      (if (> _ping_frequency 0)
        (CallFunction|StartHidingHintSound)
        (else
          (CallFunction|ClearHintTimers))))))

(event Custom|FinishPropSetup
  (bind _returnvalue (Game|GetInstigatorController))
  (bind _copiedprop (Variables|Default|GetCopiedProp))
  (Utilities|FlowControl|Sequence
    (:then_0
      (if (Networking|IsServer)
        (Variables|Visibility|SetbShouldBeHiding true)
        (PlayerPropAttachment|AttachProp (Variables|Default|GetCopiedProp))
        (Transformation|SetActorRelativeRotation (Variables|Default|GetCopiedProp) "0, 0, 0" false false)
        (bind _returnvalue_2 (Development|GetConsoleVariableBoolValue "Fort.Rollback.FixPropOffsetWhenPlaced"))
        (if _returnvalue_2
          ;[0.6348 1 0.5321 1 12] FORT- 735572 Fix props being offset when placed.
          ;; Calculate Prop Offset
          (Transformation|SetActorLocation (Variables|Default|GetCopiedProp) (+ (- (Transformation|GetWorldLocation (Variables|/Script/Engine.Character|GetCapsuleComponent (Variables|Default|GetPawnAttachedTo))) (PlayerPropAttachment|CalcWorldSpaceBounds (Variables|Default|GetCopiedProp))) (Transformation|GetActorLocation (Variables|Default|GetCopiedProp))) false false)
          (else
            (Transformation|SetActorRelativeLocation "0, 0, 0")
            (Transformation|AddActorWorldOffset (CallFunction|CalcPropOffset))))
        (bind (_settings _exists) (CallFunction|GetPropHuntSettings))
        (if _exists
          (Utilities|IsValid _returnvalue
            (:Is%20Valid
              (CreativePropPossession|OnPlayerFinishEnteringDisguise(Message) (Utilities|ToObject(Interface) _settings) _returnvalue))))))
    (:then_1
      (CallFunction|AddPropstoItemsSpecificBlackList))
    (:then_2
      (bind _returnvalue_3 (FortPhysicsSimulationLibrary|IsPhysicsObjectActiveinSimulation _copiedprop))
      (if _returnvalue_3
        (FortPhysicsSimulationLibrary|StopSimulatingPhysicsObject _copiedprop)))))

(event Custom|SetScaleTimers
  (bind _pawnattachedto (Variables|Default|GetPawnAttachedTo))
  ;[1 1 1] MulticastSetStaticMesh
  (Utilities|IsValid _pawnattachedto
    (:Is%20Valid
      (if (Variables|PlayerSize|GetbScalePawnCapsuleWithPropSize)
        (if (Variables|/Script/Engine.Character|GetbIsCrouched _pawnattachedto)
          (Character|UnCrouch _pawnattachedto)
          (bind _returnvalue (Utilities|Time|SetTimerbyEvent (EventDispatchers|CreateEvent :!event Custom|InitScaleLerp) 0.01)) ; Give CharacterMovement a frame to update
          (Variables|PlayerSize|SetInitScaleLerpTimerHandle _returnvalue)
          (bind _currentweapon (Variables|/Script/FortniteGame.FortPawn|GetCurrentWeapon (Variables|Default|GetPawnAttachedTo))
            (:then
              (FortWeapon|SetWeaponVisibility _currentweapon false true))
            (:else))
          (else
            (CallFunction|InitScaleLerp)))
        (else
          (bind _returnvalue_2 (Utilities|Time|SetTimerbyEvent (EventDispatchers|CreateEvent :!event FinishScaleLerp) (Variables|Default|GetScaledLerpDuration))) ; Since we aren't scaling the pawn capsule, set a timer to perfom post-scaling setup
          (Variables|PlayerSize|SetFinishScaleLerpTimerHandle _returnvalue_2))))))

(event Custom|InitScaleLerp
  (bind _boundsextent (Variables|Default|GetBoundsExtent))
  (bind _z (.z _boundsextent))
  ;[1 1 1] InitScaleLerp
  (Variables|PlayerSize|SetbHasScalingResetData true)
  (bind _capsulecomponent (Variables|/Script/Engine.Character|GetCapsuleComponent (Variables|Default|GetPawnAttachedTo))
    (:then
      (Variables|PlayerSize|SetOldCapsuleRadius (Components|Capsule|GetUnscaledCapsuleSize _capsulecomponent))
      (bind _output_get (Variables|PlayerSize|SetOldCapsuleHeight (.OutHalfHeight (Components|Capsule|GetUnscaledCapsuleSize _capsulecomponent))))
      (Variables|PlayerSize|SetScalePercent (/ _z _output_get))
      (Variables|PlayerSize|SetNewCapsuleRadius (Math|Float|Min(Float) (.x _boundsextent) (.y _boundsextent) _z))
      (Variables|PlayerSize|SetFinalScale (* (Variables|PlayerSize|GetOldScale) (Variables|PlayerSize|GetScalePercent)))
      (CallFunction|SynctoHintTimer))
    (:else)))

(event MinigameStartedDelegate(FortMinigameLogic)
  (if (Networking|IsServer)
    (Ability|SendGameplayEventtoActor (Variables|Actor|GetInstigator) (Variables|Default|GetPropDestroyedEventTag))))

(event MinigameEndedDelegate(FortMinigameLogic)
  (if (Networking|IsServer)
    (Ability|SendGameplayEventtoActor (Variables|Actor|GetInstigator) (Variables|Default|GetPropDestroyedEventTag))))

(event EventEndPlay
  (params (EndPlayReason :type EEndPlayReason%20Enum :default "Destroyed"))
  (|CallParentFunction EndPlayReason)
  (CallFunction|EndPropPossession)
  (bind _asminigame_player_created (Utilities|Casting|CastToMinigame_PlayerCreated (Minigame|GetMinigame (Variables|/Script/Engine.Pawn|GetPlayerState (Variables|Default|GetPawnAttachedTo)) :!class /Script/FortniteGame.MinigameBlueprintLibrary)))
  (Default|UnbindEventfromOnPropHuntSettingsChanged _asminigame_player_created (EventDispatchers|CreateEvent :!event Custom|OnPropHuntSettingsChanged)))

(event EventOnDeathServer
  (params (Damage :type Float%20%28single-precision%29) (DamageTags :type Gameplay%20Tag%20Container%20Structure%20%28by%20ref%29) (Momentum :type Vector :default "0, 0, 0") (HitInfo :type Hit%20Result%20Structure%20%28by%20ref%29) (InstigatedBy :type Controller%20Object%20Reference) (DamageCauser :type Actor%20Object%20Reference) (EffectContext :type Gameplay%20Effect%20Context%20Handle%20Structure))
  (|CallParentFunction Damage DamageTags Momentum HitInfo InstigatedBy DamageCauser EffectContext)
  (CallFunction|EndPropPossession)
  (bind _asminigame_player_created (Utilities|Casting|CastToMinigame_PlayerCreated (Minigame|GetMinigame (Variables|/Script/Engine.Pawn|GetPlayerState (Variables|Default|GetPawnAttachedTo)) :!class /Script/FortniteGame.MinigameBlueprintLibrary)))
  (Default|UnbindEventfromOnPropHuntSettingsChanged _asminigame_player_created (EventDispatchers|CreateEvent :!event Custom|OnPropHuntSettingsChanged)))

(event Custom|TransitionVFX
  (params (Direction :type ETimelineDirection%20Enum :default "Forward"))
  ;[1 1 1] Original stuff Bill K. had wired up.  Left heere in case he needs it when he updates the FX.
  (if (> (Variables|PlayerSize|GetScaleLerpDuration) 0.0)
    (Networking|CanExecuteCosmeticEvents
      (:True
        (CallFunction|InitializeMIDs)
        (switch Utilities|FlowControl|Switch|SwitchonETimelineDirection Direction
          (:Forward
            (Effects|SpawnEmitterAttached "/Game/Creative/Items/Gameplay/Flag/P_Creative_Flag_ResIn.P_Creative_Flag_ResIn" (Variables|/Script/Engine.Character|GetMesh (Variables|Default|GetPawnAttachedTo)))
            (bind (_direction _transition) (Timeline|WPOTransitionTL PlayFromStart)
              (:Update
                (for _array_element (Variables|Default|GetIntermediateMIDs)
                  (Rendering|Material|SetScalarParameterValue _array_element "WPOTransition" _transition :!class /Script/Engine.MaterialInstanceDynamic)))
              (:Finished
                _)))
          (:Backward
            (Timeline|WPOTransitionTL ReverseFromEnd))))))
  (for _array_element_2 (Variables|Default|GetStaticMeshOriginalMaterials)
    (Utilities|IsValid (Utilities|Struct|BreakCreativePlayerPropOriginalMaterials _array_element_2)
      (:Is%20Valid
        (Utilities|IsValid (.MaterialInterface_6_1E229FD4495D6813EEE733814F91F13C (Utilities|Struct|BreakCreativePlayerPropOriginalMaterials _array_element_2))
          (:Is%20Valid
            (Rendering|Material|SetMaterial (Utilities|Struct|BreakCreativePlayerPropOriginalMaterials _array_element_2) (.MaterialID_9_9ECE432E404ACD3DA7D2499E1AEDF713 (Utilities|Struct|BreakCreativePlayerPropOriginalMaterials _array_element_2)) (.MaterialInterface_6_1E229FD4495D6813EEE733814F91F13C (Utilities|Struct|BreakCreativePlayerPropOriginalMaterials _array_element_2))))))))
  (bind (_direction_2 _lerp) (Timeline|TransitionToIntermediateTL PlayFromStart)
    (:Update
      (for _array_element_3 (Variables|Default|GetPropMeshMIDs)
        (Rendering|Material|SetScalarParameterValue _array_element_3 "BlendToIntermediate" _lerp :!class /Script/Engine.MaterialInstanceDynamic)))))

(event Custom|Terminate
  (bind _returnvalue (Game|GetInstigatorController))
  (CallFunction|TransitionVFX "Forward")
  (CallFunction|EndPropPossession)
  (CallFunction|PlayPropRemovedSound)
  (if (Networking|IsServer)
    (bind (_settings _exists) (CallFunction|GetPropHuntSettings))
    (if _exists
      (Utilities|IsValid _returnvalue
        (:Is%20Valid
          (CreativePropPossession|OnPlayerExitingDisguise(Message) (Utilities|ToObject(Interface) _settings) _returnvalue))))))

(event Custom|ServerSetPlaced
  (params (NewPlaced :type Boolean) (PlacedYawOffset :type Float%20%28double-precision%29))
  ;[1 1 1] ServerSetPlaced - Since the server does not play animations, our prop rotation will always be offset from what clients see at the time of placement. We send the PlacedYaw from the locally controlled client so clients don't experience a rotation pop when entering "placed" mode
  (CallFunction|SetPropPlaced NewPlaced PlacedYawOffset))

(event Custom|AddPropsToItemsSpecificBlackList
  (bind _gameplaytagtodisablegravygoblin (Variables|Default|GetGameplayTagToDisableGravyGoblin))
  ;[1 1 1] Disable specific items (GG) from effecting disguise props
  (Ability|GameplayEffect|AddLooseGameplayTags (Variables|Default|GetCopiedProp) _gameplaytagtodisablegravygoblin)
  (Ability|GameplayEffect|AddLooseGameplayTags self _gameplaytagtodisablegravygoblin))

(event Custom|PlayHintSoundOnClient
  (CallFunction|PlayHintSound))

(fn CalcPropOffset
  (bind (_bblockinghit _bwalkablefloor _blinetrace _floordist) (Utilities|Struct|BreakFindFloorResult (Variables|/Script/Engine.CharacterMovementComponent|GetCurrentFloor (Actor|GetComponentbyClass (Variables|Default|GetPawnAttachedTo) "/Script/Engine.CharacterMovementComponent"))))
  (bind (_origin _extent) (PlayerPropAttachment|CalcWorldSpaceBounds (Variables|Default|GetCopiedProp)))
  (return (Math|Vector|MakeVector (.x (- (Transformation|GetActorLocation) _origin)) (.y (- (Transformation|GetActorLocation) _origin)) (- (+ (+ (- (.z (Transformation|GetActorLocation)) (.z _origin)) (.z _extent)) (* (Math|Float|SelectFloat (Components|Capsule|GetScaledCapsuleHalfHeight (Variables|/Script/Engine.Character|GetCapsuleComponent (Variables|Default|GetPawnAttachedTo))) (Animation|Utilities|GetDistanceBetweenTwoSockets (Variables|/Script/Engine.Character|GetMesh (Variables|Default|GetPawnAttachedTo)) "SM_Pelvis" "RTS_World" "Root_Effect") (Variables|Default|GetIsPlaced)) -1.0)) _floordist))))

(fn CalculatePropVolume
  (return (* (* (.x (Variables|Default|GetBoundsExtent)) (.y (Variables|Default|GetBoundsExtent))) (.z (Variables|Default|GetBoundsExtent)))))

(fn ClearHintTimers
  (if (Networking|IsServer)
    (Utilities|Time|ClearandInvalidateTimerbyHandle (Variables|Default|GetUICountdownTimerHandle))
    (else
      (Utilities|Time|ClearandInvalidateTimerbyHandle (Variables|Default|GetHintSoundTimer)))))

(fn DetermineNewMaxHealth
  (params (SizeInt :type Integer))
  (switch int SizeInt
    (:0
      (bind _output_get (Variables|Health|SetVolumeToHealthFactor (Variables|Health|GetVolumeToHealthFactorSmallProp)))
      (return (Math|Conversions|ToFloat(Integer) (Math|Float|Ceil (Math|Float|Clamp(Float) (* (Variables|Health|GetMeshVolume) _output_get) (Variables|Health|GetMinClampSmallProp) (Variables|Health|GetMaxClampSmallProp))))))
    (:1
      (bind _output_get_2 (Variables|Health|SetVolumeToHealthFactor (Variables|Health|GetVolumeToHealthFactorMediumProp)))
      (return (Math|Conversions|ToFloat(Integer) (Math|Float|Ceil (Math|Float|Clamp(Float) (* (Variables|Health|GetMeshVolume) _output_get_2) (Variables|Health|GetMinClampMediumProp) (Variables|Health|GetMaxClampMediumProp))))))
    (:2
      (bind _output_get_3 (Variables|Health|SetVolumeToHealthFactor (Variables|Health|GetVolumeToHealthFactorLargeProp)))
      (return (Math|Conversions|ToFloat(Integer) (Math|Float|Ceil (Math|Float|Clamp(Float) (* (Variables|Health|GetMeshVolume) _output_get_3) (Variables|Health|GetMinClampLargeProp) (Variables|Health|GetMaxClampLargeProp))))))))

(fn DoPawnRelatedInitialization
  (if (not (Variables|Default|GetbDidPawnRelatedInitialization))
    (Variables|Default|SetbDidPawnRelatedInitialization true)
    (bind _pawnattachedto (Variables|Default|GetPawnAttachedTo)
      (:then
        (CallFunction|SetTeam (Team|GetActorTeam _pawnattachedto))
        (Variables|Default|SetDefaultPawnMeshCollisionEnabled (Collision|GetCollisionEnabled (Variables|/Script/Engine.Character|GetMesh (Variables|Default|GetPawnAttachedTo))))
        (bind _capsulecomponent (Variables|/Script/Engine.Character|GetCapsuleComponent (Variables|Default|GetPawnAttachedTo))
          (:then
            (Collision|IgnoreActorwhenMoving _capsulecomponent self true)
            (bind _copiedprop (Variables|Default|GetCopiedProp)
              (:then
                (Collision|IgnoreActorwhenMoving _capsulecomponent _copiedprop true))
              (:else)))
          (:else)))
      (:else))))

(fn EndPropPossession
  (bind _pawnattachedto (Variables|Default|GetPawnAttachedTo))
  (bind _capsulecomponent (Variables|/Script/Engine.Character|GetCapsuleComponent _pawnattachedto))
  (bind (_outradius _outhalfheight) (Components|Capsule|GetUnscaledCapsuleSize _capsulecomponent))
  (bind (_outradius_2 _outhalfheight_2) (Components|Capsule|GetScaledCapsuleSize _capsulecomponent))
  (bind _oldcapsuleradius (Variables|PlayerSize|GetOldCapsuleRadius))
  (bind _copiedprop (Variables|Default|GetCopiedProp))
  (bind _pawnattachedto_2 (Variables|Default|GetPawnAttachedTo))
  (if (not (Variables|Default|GetbHasCleanedUp))
    (PlayerPropAttachment|LogState "BP - Pre EndPropPossession()")
    (Variables|Default|SetbHasCleanedUp true)
    (Utilities|FlowControl|Sequence
      (:then_0
        (Variables|Visibility|SetbShouldBeHiding false)
        (Utilities|Time|ClearandInvalidateTimerbyHandle (Variables|Visibility|GetVisibilityRecheckTimerHandle))
        (Utilities|Time|ClearandInvalidateTimerbyHandle (Variables|PlayerSize|GetInitScaleLerpTimerHandle))
        (Utilities|Time|ClearandInvalidateTimerbyHandle (Variables|PlayerSize|GetFinishScaleLerpTimerHandle))
        (Utilities|Time|ClearandInvalidateTimerbyHandle (Variables|PlayerSize|GetScaleLerpTimerHandle))
        (Transformation|DetachFromActor "KeepWorld" "KeepWorld" "KeepWorld")
        (Variables|Default|GetPawnAttachedTo
          (:then
            (GameplayEffects|RemoveActiveGameplayEffect (Ability|GetAbilitySystemComponent (Variables|Default|GetPawnAttachedTo)) (Variables|Default|GetGE_SetMaxHealthActive) -1)
            (Utilities|FlowControl|Sequence
              (:then_0
                (if (Variables|PlayerSize|GetbHasScalingResetData)
                  (Utilities|FlowControl|Sequence
                    (:then_0
                      ;[0.5573 0.5573 0.5573] If we're increasing our capsule size, push our capsule upward so its new size will fit and not potentially get pushed down through the ground
                      (if (or (> _oldcapsuleradius _outradius_2) (> _outhalfheight _outhalfheight_2))
                        (Transformation|AddActorWorldOffset (Variables|Default|GetPawnAttachedTo) (Math|Vector|MakeVector 0.0 0.0 (+ (Math|Float|Max(Float) (- _oldcapsuleradius _outradius_2)) (Math|Float|Max(Float) (- _outhalfheight _outhalfheight_2)))) false false)))
                    (:then_1
                      (Transformation|SetActorScale3D (Variables|Default|GetPawnAttachedTo) (Variables|PlayerSize|GetOldScale))
                      (bind _capsulecomponent_2 (Variables|/Script/Engine.Character|GetCapsuleComponent (Variables|Default|GetPawnAttachedTo))
                        (:then
                          (Components|Capsule|SetCapsuleSize _capsulecomponent_2 (Variables|PlayerSize|GetOldCapsuleRadius) (Variables|PlayerSize|GetOldCapsuleHeight) true))
                        (:else)))
                    (:then_2
                      (bind _charactermovement (Variables|/Script/Engine.Character|GetCharacterMovement (Variables|Default|GetPawnAttachedTo))
                        (:then
                          (Variables|/Script/Engine.CharacterMovementComponent|SetbForceNextFloorCheck true _charactermovement))
                        (:else))))))
              (:then_1
                (bind _currentweapon (Variables|/Script/FortniteGame.FortPawn|GetCurrentWeapon _pawnattachedto_2)
                  (:then
                    (FortWeapon|SetWeaponVisibility _currentweapon (not (Weapon|IsWeaponHolstered _pawnattachedto_2)) true))
                  (:else)))))
          (:else)))
      (:then_1
        (Variables|PlayerSize|SetbHasScalingResetData false)
        (Utilities|FlowControl|Sequence
          (:then_0
            (bind _fortminigamelogic (Variables|Default|GetFortMinigameLogic)
              (:then
                (bind _currentminigame (Variables|/Script/FortniteGame.FortMinigameLogicComponent|GetCurrentMinigame _fortminigamelogic)
                  (:then
                    (Minigame|RemoveMinigameActor _currentminigame self))
                  (:else)))
              (:else)))
          (:then_1
            (Variables|Default|GetCopiedProp
              (:then
                (Utilities|FlowControl|Sequence
                  (:then_0
                    (if (Utilities|DoesObjectImplementInterface (Variables|Default|GetCopiedProp) "/Game/Weapons/Prototype/ObjectMover/Blueprints/ObjectInteractionBehaviors/PossessProp/Creative_Player_PropInterface.Creative_Player_PropInterface_C")
                      (CreativePlayerProp|PreDestroy(Message) (Variables|Default|GetCopiedProp))))
                  (:then_1
                    (if (Variables|Default|GetbUseSilentDie)
                      (bind _asbuilding_actor (Utilities|Casting|CastToBuildingActor _copiedprop)
                        (:then
                          (Gameplay|SilentDie _asbuilding_actor false))
                        (:CastFailed
                          _)))
                    (Actor|DestroyActor _copiedprop))))
              (:else)))
          (:then_2
            (if (Variables|Default|GetbUseSilentDie)
              (Gameplay|SilentDie false)
              (else
                (Actor|DestroyActor))))))
      (:then_2
        (Variables|Default|SetbDidPawnRelatedInitialization false)))))

(fn FinishScaleLerp
  (Utilities|Time|ClearandInvalidateTimerbyHandle (Variables|PlayerSize|GetScaleLerpTimerHandle))
  (CallFunction|SetScale (Variables|PlayerSize|GetFinalScale))
  (CallFunction|SetCapsuleRadius)
  (Variables|/Script/Engine.CharacterMovementComponent|SetbForceNextFloorCheck true (Variables|/Script/Engine.Character|GetCharacterMovement (Variables|Default|GetPawnAttachedTo)))
  (Rendering|SetVisibility (Variables|Default|GetScene) true true)
  (CallFunction|FinishPropSetup)
  (CallFunction|TransitionVFX "Backward")
  (Variables|Default|SetPrePlacementRelativeTransform (Math|Transform|MakeTransform (Variables|/Script/Engine.SceneComponent|GetRelativeLocation (Variables|Transformation|GetRootComponent)) (Variables|Default|GetPrePlacementRotation) (.scale (Variables|Default|GetPrePlacementRelativeTransform))))
  (Transformation|SetActorScale3D (Variables|Default|GetCopiedProp) (Variables|Default|GetOriginalPropScale)) ; Reset copied prop size because it could have a wrong size before attaching. Keep World rule will also change it's size
  (Transformation|AttachActorToActor (Variables|Default|GetCopiedProp) self "None" "SnapToTarget" "KeepWorld" "KeepWorld" true)
  (Transformation|SetActorRelativeRotation (Variables|Default|GetCopiedProp) "0, 0, 0" false false)
  (Transformation|SetActorRelativeRotation "0, 0, 0"))

(fn GetPropHuntSettings
  (bind _returnvalue (Game|GetInstigatorController))
  (Utilities|IsValid _returnvalue
    (:Is%20Valid
      (bind _asminigame_player_created (Utilities|Casting|CastToMinigame_PlayerCreated (Minigame|GetMinigame (Variables|/Script/Engine.Controller|GetPlayerState _returnvalue) :!class /Script/FortniteGame.MinigameBlueprintLibrary)))
      (bind _asprop_possessor_config_interface (Utilities|Casting|CastToPropPossessorConfigInterface (Variables|/Game/Athena/Playset/Minigames/Minigame_PlayerCreated.Minigame_PlayerCreated_C|GetPropHuntSettings _asminigame_player_created)))
      (return _asprop_possessor_config_interface true))))

(fn GetRelativePropSizeForHealthScaling
  (params (PropVolume :type Float%20%28double-precision%29))
  ;[1 1 1] Returns an int that represents one of three possible relative sizes 2=Large, 1=medium, 0=small
  (Math|Float|CompareFloat PropVolume 3e+06
    (:>
      _)
    (:==
      _)
    (:<
      (Math|Float|CompareFloat PropVolume 1e+06
        (:>
          _)
        (:==
          _)
        (:<
          (return 0)))
      (return 1)))
  (return 2))

(fn GetRelativePropSizeForHintSoundTiming
  (params (PropVolume :type Float%20%28double-precision%29))
  ;[1 1 1] Returns an int that represents one of three possible relative sizes 3, Extra Lrg 2=Large, 1=medium, 0=small
  (Math|Float|CompareFloat PropVolume 1e+06
    (:>
      _)
    (:==
      _)
    (:<
      (Math|Float|CompareFloat PropVolume 1.5e+05
        (:>
          _)
        (:==
          _)
        (:<
          (Math|Float|CompareFloat PropVolume 1e+04
            (:>
              _)
            (:==
              _)
            (:<
              (return 0)))
          (return 1)))
      (return 2)))
  (return 3))

(fn InitializeMIDs
  (Utilities|Array|Clear (Variables|Default|GetPropMeshMIDs))
  (Utilities|Array|Clear (Variables|Default|GetStaticMeshOriginalMaterials))
  (for _array_element (Actor|GetComponentsByClass (Variables|Default|GetCopiedProp) "/Script/Engine.StaticMeshComponent")
    (LocalVariables|SetL_CurrentComponent _array_element)
    (for (_array_element_2 _array_index) (Rendering|Material|GetMaterials _array_element)
      (LocalVariables|SetL_OriginalMaterial _array_element_2)
      (LocalVariables|SetL_CurrentMaterialIndex _array_index)
      (Utilities|FlowControl|Sequence
        (:then_0
          ;[1 1 1] Generate an MID if it does not have one
          (Utilities|Casting|CastToMaterialInstanceDynamic (LocalVariables|GetL_OriginalMaterial)
            (:CastFailed
              (Rendering|Material|CreateDynamicMaterialInstance (LocalVariables|GetL_CurrentComponent) (LocalVariables|GetL_CurrentMaterialIndex) 0 "None" :!class /Script/Engine.PrimitiveComponent))))
        (:then_1
          ;[1 1 1] Add the current material to the list of original materials
          (bind _asmaterial_instance_dynamic (Utilities|Casting|CastToMaterialInstanceDynamic (Rendering|Material|GetMaterial (LocalVariables|GetL_CurrentComponent) (LocalVariables|GetL_CurrentMaterialIndex)))
            (:then
              (Utilities|Array|AddUnique (Variables|Default|GetPropMeshMIDs) _asmaterial_instance_dynamic))
            (:CastFailed
              _))
          (Utilities|Array|AddUnique (Variables|Default|GetStaticMeshOriginalMaterials) (Utilities|Struct|MakeCreativePlayerPropOriginalMaterials (LocalVariables|GetL_CurrentComponent) _asmaterial_instance_dynamic (LocalVariables|GetL_CurrentMaterialIndex))))
        (:then_2
          (bind _returnvalue (Rendering|Material|GetBaseMaterial (LocalVariables|GetL_OriginalMaterial)))
          (if (or (Utilities|Enum|Equal(Enum) (Variables|/Script/Engine.Material|GetBlendMode _returnvalue) "BLEND_Opaque") (Utilities|Enum|Equal(Enum) (Variables|/Script/Engine.Material|GetBlendMode _returnvalue) "BLEND_Masked"))
            _
            (elif (Utilities|Array|ContainsItem (Variables|Default|GetNonOpaqueMaterialsToRender) (Utilities|ResolveSoftReference _returnvalue))
              _
              (else
                (Rendering|Material|SetMaterial (LocalVariables|GetL_CurrentComponent) (LocalVariables|GetL_CurrentMaterialIndex) "/Game/Weapons/Prototype/ObjectMover/Materials/M_MoveTool_Invisible.M_MoveTool_Invisible"))))
          (bind _returnvalue_2 (Rendering|Material|CreateDynamicMaterialInstance (LocalVariables|GetL_CurrentComponent) (LocalVariables|GetL_CurrentMaterialIndex) "/Game/Weapons/Prototype/ObjectMover/Materials/M_CreativeMoveTool_Intermediate.M_CreativeMoveTool_Intermediate" "None" :!class /Script/Engine.PrimitiveComponent))
          (Utilities|Array|Add (Variables|Default|GetIntermediateMIDs) _returnvalue_2))))))

(fn OnRep_bShouldBeHiding
  (CallFunction|UpdateVisibility))

(fn OnRep_CopiedProp
  (for _array_element (Actor|GetComponentsByClass (Variables|Default|GetCopiedProp) "/Script/Engine.PrimitiveComponent")
    (Transformation|SetMobility _array_element "Movable")
    (bind _asstatic_mesh_component (Utilities|Casting|CastToStaticMeshComponent _array_element)
      (:then
        (if (not (PlayerPropAttachment|DoesMeshComponentHaveVisibleMaterial _asstatic_mesh_component))
          (Collision|SetCollisionProfileName _array_element "NoCollision" true :!class /Script/Engine.PrimitiveComponent)))
      (:CastFailed
        (bind _asgeometry_collection_component (Utilities|Casting|CastToGeometryCollectionComponent _array_element)
          (:then
            (ChaosPhysics|SetPerLevelCollisionProfileNames _asgeometry_collection_component (Utilities|Array|MakeArray "FortPossessedProp")))
          (:CastFailed
            _))))
    (Collision|SetCollisionProfileName _array_element "FortPossessedProp" true :!class /Script/Engine.PrimitiveComponent)
    (Variables|/Script/Engine.PrimitiveComponent|SetCanCharacterStepUpOn "ECB_Yes" _array_element))
  (bind _asbuilding_actor (Utilities|Casting|CastToBuildingActor (Variables|Default|GetCopiedProp))
    (:then
      (DrawDistance|SetBRDrawDistances _asbuilding_actor 0.0 (Variables|/Script/Engine.PrimitiveComponent|GetCachedMaxDrawDistance (Variables|/Script/Engine.Character|GetMesh (Variables|Default|GetPawnAttachedTo)))))
    (:CastFailed
      _))
  (Networking|CanExecuteCosmeticEvents
    (:True
      (CallFunction|DoPawnRelatedInitialization)
      (CallFunction|UpdateVisibility)
      (bind _asbuilding_actor_2 (Utilities|Casting|CastToBuildingActor (Variables|Default|GetCopiedProp)))
      (CallFunction|SetSuppressHealthBar _asbuilding_actor_2 true)
      (CallFunction|SetScaleTimers)
      (PlayerPropAttachment|LogState "Post BP - OnRep_CopiedProp"))))

(fn OnRep_FinalScale
  (CallFunction|UpdateVisibility)
  (if (not (Variables|Default|GetbHasCleanedUp))
    (bind _returnvalue (Utilities|Time|SetTimerbyFunctionName self (Variables|PlayerSize|GetScaleLerpName) (Variables|PlayerSize|GetScaleLerpInterval) true))
    (Variables|PlayerSize|SetScaleLerpTimerHandle _returnvalue)))

(fn OnRep_NewCapsuleRadius
  (CallFunction|SetCapsuleRadius)
  (CallFunction|UpdateVisibility))

(fn OnRep_PawnAttachedTo
  (CallFunction|DoPawnRelatedInitialization)
  (CallFunction|SetCapsuleRadius)
  (CallFunction|UpdateVisibility)
  (PlayerPropAttachment|LogState "Post BP - OnRep_PawnAttachedTo"))

(fn OnRep_ScalePercent
  (CallFunction|SetCapsuleRadius)
  (CallFunction|UpdateVisibility))

(fn OnRep_SoundTimeRemaining
  (bind _returnvalue (BlueprintContext|GetCreativeHUDLContext))
  (Networking|CanExecuteCosmeticEvents
    (:True
      (Utilities|IsActorRelevanttoMe (Class|CreativeHUDLContext|GetOwningPlayerState _returnvalue)
        (:IsRelevant
          (EventDispatchers|CallOnPropPlaySoundCountdownChanged _returnvalue (Variables|Default|GetCountdownText) (Variables|Default|GetSoundTimeRemaining)))))))

(fn PlayPropRemovedSound
  (bind _pawnattachedto (Variables|Default|GetPawnAttachedTo)
    (:then
      (Audio|SpawnSoundAttached (select (FortPawn|IsDead _pawnattachedto) (Variables|Audio|GetPlayer%20Eliminated%20Sound) (Variables|Audio|GetBackToPlayerSound)) (Variables|Transformation|GetRootComponent (Variables|Default|GetPawnAttachedTo))))
    (:else)))

(fn ScaleLerp
  (bind (_finalscale_x _finalscale_y _finalscale_z) (Math|Vector|BreakVector (Variables|GetFinalScale_X)))
  (bind (_settings _exists) (CallFunction|GetPropHuntSettings))
  (if _exists
    (bind (_override _duration) (CreativePropPossession|GetDisguiseAnimationOverrideDetails _settings))
    (if _override
      (Variables|PlayerSize|SetScaleLerpDuration _duration)))
  (if (> (Variables|PlayerSize|GetScaleLerpDuration) 0.0)
    (if (>= (Variables|PlayerSize|GetScaleLerpAlpha) 1.0)
      (CallFunction|FinishScaleLerp)
      (else
        (bind _pawnattachedto (Variables|Default|GetPawnAttachedTo)
          (:then
            (Transformation|SetActorScale3D _pawnattachedto (Math|Vector|SelectVector (Variables|PlayerSize|GetOldScale) (Math|Vector|Lerp(Vector) (Variables|PlayerSize|GetOldScale) (Variables|PlayerSize|GetFinalScale) (Variables|PlayerSize|GetScaleLerpAlpha)) (Variables|PlayerSize|GetbSkipScalingAnimation)))
            (Variables|PlayerSize|SetScaleLerpTimeElapsed (+ (Variables|PlayerSize|GetScaleLerpTimeElapsed) (Variables|PlayerSize|GetScaleLerpInterval)))
            (Variables|PlayerSize|SetScaleLerpAlpha (/ (Variables|PlayerSize|GetScaleLerpTimeElapsed) (Variables|PlayerSize|GetScaleLerpDuration)))
            (bind _charactermovement (Variables|/Script/Engine.Character|GetCharacterMovement (Variables|Default|GetPawnAttachedTo))
              (:then
                (Variables|/Script/Engine.CharacterMovementComponent|SetbForceNextFloorCheck true _charactermovement))
              (:else)))
          (:else))))
    (else
      ;[1 1 1] Handle case where lerp duration wasn't recieved (is zero), likely we were far from the player when possessing and thus shouldn't lerp the scale as it would give us away
      (bind _pawnattachedto_2 (Variables|Default|GetPawnAttachedTo)
        (:then
          (Transformation|SetActorScale3D _pawnattachedto_2 (Variables|PlayerSize|GetFinalScale))
          (CallFunction|FinishScaleLerp))
        (:else))))
  ;[1 1 1] Push character up by the difference in our scale, to avoid falling through the world
  (if (> _finalscale_z 1.0)
    (Transformation|AddActorWorldOffset (Variables|Default|GetPawnAttachedTo) (Math|Vector|MakeVector 0.0 0.0 (Math|Float|float*float _finalscale_z (Components|Capsule|GetUnscaledCapsuleHalfHeight (Variables|/Script/Engine.Character|GetCapsuleComponent (Variables|Default|GetPawnAttachedTo))) (/ (Variables|PlayerSize|GetScaleLerpInterval) (Variables|PlayerSize|GetScaleLerpDuration)))) true false)))

(fn SetCapsuleRadius
  (if (and (> (Variables|PlayerSize|GetScalePercent) 0.0) (> (Variables|PlayerSize|GetNewCapsuleRadius) 0.0))
    (bind _pawnattachedto (Variables|Default|GetPawnAttachedTo)
      (:then
        (bind _capsulecomponent (Variables|/Script/Engine.Character|GetCapsuleComponent _pawnattachedto)
          (:then
            (Components|Capsule|SetCapsuleRadius _capsulecomponent (* (Variables|PlayerSize|GetNewCapsuleRadius) (/ 1.0 (Variables|PlayerSize|GetScalePercent))) true))
          (:else)))
      (:else))))

(fn SetMaxHealth
  (params (NewMaxHealth :type Float%20%28double-precision%29))
  (Networking|SwitchHasAuthority
    (:Authority
      (Variables|Default|GetPawnAttachedTo
        (:then
          (bind _returnvalue (Ability|GameplayEffect|AssignTagSetbyCallerMagnitude (GameplayEffects|MakeOutgoingSpec :self (Ability|GetAbilitySystemComponent (Variables|Default|GetPawnAttachedTo)) :GameplayEffectClass (Variables|Default|GetGE_SetMaxHealthClass) :Level 1.0) (LocalVariables|GetMaxHealthDataTag) NewMaxHealth))
          (bind _returnvalue_2 (GameplayEffects|ApplyGameplayEffectSpecToSelf (Ability|GetAbilitySystemComponent (Variables|Default|GetPawnAttachedTo)) _returnvalue))
          (Variables|Default|SetGE_SetMaxHealthActive _returnvalue_2)
          (GameplayEffects|ApplyGameplayEffectSpecToTarget (Ability|GetAbilitySystemComponent (Variables|Default|GetPawnAttachedTo)) _returnvalue (Ability|GetAbilitySystemComponent self)))
        (:else)))))

(fn SetProp
  (params (SelectedProp :type Actor%20Object%20Reference) (PlayerPawnToAttachTo :type Player%20Pawn%20Athena%20Object%20Reference) (ScaleDuration :type Float%20%28double-precision%29))
  (bind (_origin _extent) (PlayerPropAttachment|CalcWorldSpaceBounds (Variables|GetSelectedProp)))
  (PlayerPropAttachment|InitProp (Variables|GetSelectedProp))
  (Fortnite|SetCanbeDamaged (Variables|GetSelectedProp))
  (Networking|SetReplicates (Variables|GetSelectedProp) true)
  (Variables|PlayerSize|SetOldScale (Transformation|GetActorScale3D (Variables|GetPlayerPawnToAttachTo)))
  (Variables|Default|SetOriginalPropScale (Transformation|GetActorScale3D (Variables|GetSelectedProp)))
  (Variables|Default|SetBoundsExtent (Math|Vector|MakeVector (Math|Float|Clamp(Float) (.x _extent) (Variables|PlayerSize|GetMinCapsuleRadius) (Variables|PlayerSize|GetMaxCapsuleRadius)) (Math|Float|Clamp(Float) (.y _extent) (Variables|PlayerSize|GetMinCapsuleRadius) (Variables|PlayerSize|GetMaxCapsuleRadius)) (Math|Float|Clamp(Float) (.z _extent) (Variables|PlayerSize|GetMinCapsuleHalfHeight) (Variables|PlayerSize|GetMaxCapsuleHalfHeight))))
  (Variables|Default|SetCopiedProp (Variables|GetSelectedProp))
  (Variables|Default|SetPawnAttachedTo (Variables|GetPlayerPawnToAttachTo))
  (CallFunction|UpdateHealth)
  (Transformation|AttachActorToComponent (Variables|/Script/Engine.Character|GetMesh (Variables|Default|GetPawnAttachedTo)) "SM_Pelvis" "SnapToTarget" "KeepWorld" "KeepWorld")
  (Variables|Default|SetScaledLerpDuration (Variables|GetScaleDuration))
  (Actor|SetOwner (Variables|Default|GetCopiedProp) self)
  (CallFunction|SetScaleTimers)
  (PlayerPropAttachment|LogState "Post BP - SetProp (Inc Attach To Pawn)"))

(fn SetPropPlaced
  (params (NewPlaced :type Boolean) (PlacedYawOffset :type Float%20%28double-precision%29))
  (bind _returnvalue (BlueprintContext|GetCreativeHUDLContext))
  (bind _output_get (Variables|Default|SetIsPlaced NewPlaced))
  (Utilities|FlowControl|Sequence
    (:then_0
      (Utilities|IsValid _returnvalue
        (:Is%20Valid
          (Class|CreativeHUDLContext|ForwardonPropLockedinPlacedChanged _returnvalue _output_get))))
    (:then_1
      (if _output_get
        (bind _output_get_2 (Variables|Default|SetPrePlacementRelativeTransform (Transformation|GetRelativeTransform (Variables|Transformation|GetRootComponent))))
        (Variables|Default|SetPrePlacementRotation (.rotation _output_get_2))
        (Transformation|AttachActorToComponent (Variables|/Script/Engine.Character|GetCapsuleComponent (Variables|Default|GetPawnAttachedTo)) "" "SnapToTarget" "KeepWorld" "KeepWorld")
        (Transformation|SetActorRelativeLocation "0, 0, 0")
        (Transformation|AddActorWorldOffset (CallFunction|CalcPropOffset))
        ;[1 1 1] Since the server does not play animations, our prop rotation will always be offset from what clients see at the time of placement. We send the PlacedYawOffset from the locally controlled client so clients don't experience a rotation pop when entering "placed" mode
        (if (Networking|IsServer)
          (Transformation|AddWorldRotation (Variables|Transformation|GetRootComponent) (Math|Rotator|MakeRotator 0.0 0.0 (Variables|GetPlacedYawOffset)) false false))
        (else
          (Transformation|AttachActorToComponent (Variables|/Script/Engine.Character|GetMesh (Variables|Default|GetPawnAttachedTo)) "SM_Pelvis" "SnapToTarget" "KeepWorld" "KeepWorld")
          (Transformation|SetActorRelativeTransform (Variables|Default|GetPrePlacementRelativeTransform)))))))

(fn SetScale
  (params (NewScale :type Vector))
  (if (> (Math|Vector|VectorLengthSquared NewScale) 0.0)
    (bind _pawnattachedto (Variables|Default|GetPawnAttachedTo)
      (:then
        (Transformation|SetActorScale3D _pawnattachedto NewScale))
      (:else))))

(fn UpdateHealth
  ;[1 1 1] Finds volume of the prop and then multiplies the health by a factor based on the relative size (sm, med, lrg)
  (bind (_settings _exists) (CallFunction|GetPropHuntSettings))
  (if _exists
    (bind (_show_props_remaining _prop_health_scaling) (CreativePropPossession|GetPropDetails _settings))
    (if _prop_health_scaling
      _))
  (bind _meshvolume (CallFunction|CalculatePropVolume))
  (bind _output_get (Variables|Health|SetMeshVolume _meshvolume))
  (bind _relativesize (CallFunction|GetRelativePropSizeforHealthScaling _output_get))
  (bind _newmaxhealth (CallFunction|DetermineNewMaxHealth _relativesize))
  (bind _output_get_2 (LocalVariables|SetNewMaxHealth _newmaxhealth))
  (Variables|Health|SetCurrentMaxHealth _output_get_2)
  (Variables|Health|SetOldMaxHealth (FortPawn|GetMaxHealth (Variables|Default|GetPawnAttachedTo)))
  (CallFunction|SetMaxHealth (LocalVariables|GetNewMaxHealth)))

(fn UpdateVisibility
  (bind _bshouldbehiding (Variables|Visibility|GetbShouldBeHiding))
  (bind _returnvalue (not _bshouldbehiding))
  (Utilities|FlowControl|Sequence
    (:then_0
      (Networking|CanExecuteCosmeticEvents
        (:True
          (bind _pawnattachedto (Variables|Default|GetPawnAttachedTo)
            (:then
              (bind _returnvalue_2 (Class|FortPlayerPawnAthena|SetCharacterBodyVisibilityforPossession _pawnattachedto (not (Variables|Visibility|GetbShouldBeHiding))))
              (if _returnvalue_2
                (bind _currentweapon (Variables|/Script/FortniteGame.FortPawn|GetCurrentWeapon (Variables|Default|GetPawnAttachedTo))
                  (:then
                    (Development|PrintString (Utilities|String|ToString(Boolean) _returnvalue))
                    (FortWeapon|SetWeaponVisibility _currentweapon _returnvalue true))
                  (:else
                    _))
                (Rendering|SetVisibility (Variables|Default|GetScene) (Variables|Visibility|GetbShouldBeHiding) true)
                (Rendering|SetActorHiddenInGame (Variables|Default|GetCopiedProp) (not (Variables|Visibility|GetbShouldBeHiding)))
                (Utilities|Time|ClearandInvalidateTimerbyHandle (Variables|Visibility|GetVisibilityRecheckTimerHandle))
                ;[1 1 1] chris.pickett: fixup visibility for replay scrubbing FORT-179920
                (bind _returnvalue_3 (Utilities|Time|SetTimerbyFunctionName 0 "VerifyPropVisibility" 0.05 true))
                (Variables|Visibility|SetVerifyVisibilityTimer _returnvalue_3)
                (else
                  _)))
            (:else
              _))
          (if (not (> (Utilities|Time|GetTimerRemainingTimebyHandle (Variables|Visibility|GetVisibilityRecheckTimerHandle)) 0.0))
            (bind _returnvalue_4 (Utilities|Time|SetTimerbyFunctionName self (LocalVariables|GetUpdateVisibilityName) 0.1)) ; Can't set the visibility until the mesh exists
            (Variables|Visibility|SetVisibilityRecheckTimerHandle _returnvalue_4)))))
    (:then_1
      (bind _pawnattachedto_2 (Variables|Default|GetPawnAttachedTo)
        (:then
          (bind _capsulecomponent (Variables|/Script/Engine.Character|GetCapsuleComponent _pawnattachedto_2)
            (:then
              (Collision|SetCollisionProfileName _capsulecomponent (select (Variables|Visibility|GetbShouldBeHiding) "FortPawnCapsulePossessedProp" "FortPawnCapsule") true :!class /Script/Engine.PrimitiveComponent))
            (:else)))
        (:else)))))

(fn ConstructionScript)

(fn VerifyPropVisibility
  (if (Variables|Visibility|GetbShouldBeHiding)
    (if (and (Utilities|IsValid (Variables|Default|GetCopiedProp) :!class /Script/Engine.KismetSystemLibrary) (Variables|Rendering|GetbHidden (Variables|Default|GetCopiedProp)))
      (Rendering|SetActorHiddenInGame (Variables|Default|GetCopiedProp) false))
    (else
      (Utilities|Time|ClearandInvalidateTimerbyHandle (Variables|Visibility|GetVerifyVisibilityTimer)))))
