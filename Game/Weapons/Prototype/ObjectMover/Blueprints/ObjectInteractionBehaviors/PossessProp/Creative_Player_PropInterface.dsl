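; Creative_Player_PropInterface is an interface in the ObjectMover prototype weapon's PossessProp interaction behavior folder, apparently for props possessed by a player.
; It declares a single function, PreDestroy, with no parameters or body.
; The name suggests a hook called before destruction, but the file shows no callers or implementers.
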(fn PreDestroy)
