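; GAT_Creative_PossessProp is a gameplay ability for possessing props in the prototype Creative ObjectMover tool.
; On activation it calls the parent implementation, then runs generic ability setup.
; From the setup's output it removes gameplay effects from the owner that granted the active prop-possessor effect tag.
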(event Ability|EventActivateAbility
  (|CallParentFunction)
  (Setup|SetupGenericAbility true false
    (:Output
      (Ability|RemoveGameplayEffectFromOwnerWithGrantedTags "(GameplayTags=((TagName=\"Creative.PropPossessor.Active.Effect\")))"))))
