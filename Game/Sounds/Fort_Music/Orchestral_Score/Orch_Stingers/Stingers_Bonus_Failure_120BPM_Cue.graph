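# Sound cue for an orchestral score music stinger, named for a bonus failure at 120 BPM.
# A single wave player feeds the Stingers_Bonus_Failure_120BPM wave straight to the output, with no other nodes.
# It uses the Fort_Music_Stinger_PSM sound class with boosted volume, a very large max distance and spatialization turned off in its attenuation overrides.
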SoundCue /Game/Sounds/Fort_Music/Orchestral_Score/Orch_Stingers/Stingers_Bonus_Failure_120BPM_Cue
root=WavePlayer_0
nodes [{id=WavePlayer_0, properties {soundWaveAssetPtr=&/Stingers_Bonus_Failure_120BPM}}]
properties {
 maxDistance=2097152
 soundClassObject=/Game/Sounds/FortSoundClassesAndModes/Fort_Music_Stinger_PSM
 volumeMultiplier=2.5
 attenuationOverrides {bSpatialize=false}
}
