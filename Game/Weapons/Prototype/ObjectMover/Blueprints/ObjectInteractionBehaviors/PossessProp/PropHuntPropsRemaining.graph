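# Minigame HUD widget for Prop Hunt that shows a "Players Hiding:" count of remaining hiders.
# When the in-progress UI displays or settings change, it checks for a Prop Hunt game and the config interface's Show Props Remaining flag, then reveals or collapses.
# While shown, minigame hider-count delegates update and pulse the number, hiding it below zero.
# A separate UpdateCount event increments ObjectivesComplete and plays a team-dependent success or failure stinger.
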WidgetBlueprint /Game/Weapons/Prototype/ObjectMover/Blueprints/ObjectInteractionBehaviors/PossessProp/PropHuntPropsRemaining
animations [
 AddKill {
  binding_count=1
  length_seconds=0.3
  object=&self:AddKill
  sequence {bindings [Remaining {
     id=#1
     display_name=KillCount
     possessed_class=CommonTextBlock
     tracks [{
       type=MovieScene2DTransformTrack
       sections [{
         type=MovieScene2DTransformSection
         end_frame=6
         start_frame=0
         channels [
          Translation.X {type=MovieSceneScriptingFloatChannel}
          Translation.Y {type=MovieSceneScriptingFloatChannel}
          Angle {type=MovieSceneScriptingFloatChannel}
          Scale.X {type=MovieSceneScriptingFloatChannel, keys [[0,2],[1,2],[3,0.8],[3.9207,0.8],[6,1]]}
          Scale.Y {type=MovieSceneScriptingFloatChannel, keys [[0,2],[1,2],[3,0.8],[3.9207,0.8],[6,1]]}
          Shear.X {type=MovieSceneScriptingFloatChannel}
          Shear.Y {type=MovieSceneScriptingFloatChannel}
         ]
        }]
      }]
    }]}
 }
 Reveal {
  binding_count=1
  length_seconds=0.55
  object=&self:Reveal
  sequence {bindings [Message {
     id=#2
     display_name=Content
     possessed_class=CommonTextBlock
     tracks [
      {
       type=MovieSceneFloatTrack
       display_name=Render Opacity
       sections [{type=MovieSceneFloatSection, end_frame=11, start_frame=0, channels [None {type=MovieSceneScriptingFloatChannel, keys [[0,0],[11,1]]}]}]
      }
      {
       type=MovieScene2DTransformTrack
       sections [{
         type=MovieScene2DTransformSection
         end_frame=11
         start_frame=0
         channels [
          Translation.X {type=MovieSceneScriptingFloatChannel, const [0,1]}
          Translation.Y {type=MovieSceneScriptingFloatChannel, keys [[0,-150],[11,0]]}
          Angle {type=MovieSceneScriptingFloatChannel, const [0,1]}
          Scale.X {type=MovieSceneScriptingFloatChannel, const [1,1]}
          Scale.Y {type=MovieSceneScriptingFloatChannel, const [1,1]}
          Shear.X {type=MovieSceneScriptingFloatChannel, const [0,1]}
          Shear.Y {type=MovieSceneScriptingFloatChannel, const [0,1]}
         ]
        }]
      }
     ]
    }]}
 }
]
graphs [
 AnimateIn {nodes [
   AnimateIn() {is=title K2Node_FunctionEntry}
   > PlayAnimation(InAnimation=Get Reveal() -> (Reveal:Widget Animation Object Reference))
  ]}
 EventGraph {nodes [
   n1 = UpdateCurrentNumberOfHiders!() -> (OutputDelegate, then, CurrentHiders:int) {is=title K2Node_CustomEvent}
   > n2 = Branch(Condition="integer < integer"!(A:int=@n1.CurrentHiders, B:int=0) -> (ReturnValue:bool) {is=title K2Node_CallFunction})
   > SetVisibility(InVisibility=Collapsed)
   SetVisibility(execute=@n2.else, InVisibility=HitTestInvisible)
   > SetCurrentHidersText!(execute, self, CurrentHiders:int=@n1.CurrentHiders) -> (then)
   UpdateCount() {is=title K2Node_CustomEvent}
   > Increment Float!(" ":Exec, Value:"Float (double-precision) (by ref)"=Get ObjectivesComplete() -> (ObjectivesComplete:double)) -> ("  ":Exec, Result:double=~) {is=title K2Node_MacroInstance}
   > "SetText (Text)"(
    self=Get Remaining() -> (Remaining:Common Text Object Reference)
    InText="To Text (Float)"(Value=Get ObjectivesComplete(), MinimumIntegralDigits=1, MaximumIntegralDigits=324, MinimumFractionalDigits=0, MaximumFractionalDigits=3)
   )
   > PlayAnimation(InAnimation=Get AddKill() -> (AddKill:Widget Animation Object Reference))
   n3 = GetActorTeam!(Actor:Actor Object Reference=GetOwningPlayerPawn!(self:User Widget Object Reference) -> (ReturnValue:Pawn Object Reference)) -> (ReturnValue:Byte=0)
   n4 = Literal enum EFortTeam!(Enum:EFortTeam Enum=Monster) -> (ReturnValue:EFortTeam Enum) {is=title K2Node_EnumLiteral}
   > n5 = Branch(Condition="Equal (Byte)"!(A:Byte=Byte - Byte!(A:Byte=@n3, B:Byte=@n4) -> (ReturnValue:Byte=0), B:Byte=1) -> (ReturnValue:bool))
   > PlaySound2D(Sound=/Game/Sounds/Fort_Music/Orchestral_Score/Orch_Stingers/Stingers_Bonus_Failure_120BPM_Cue)
   PlaySound2D(execute=@n5.else, Sound=/Game/Sounds/Fort_Music/Orchestral_Score/Orch_Stingers/Stingers_Bonus_Success_120BPM_Cue)
   n6 = OnPropHuntSettingsChanged!() -> (OutputDelegate, then, Minigame:Minigame Player Created Object Reference) {is=title K2Node_CustomEvent}
   n7 = Get PropPossessorConfig() -> (PropPossessorConfig:Prop Possessor Config Interface Interface)
   > n8 = Branch(Condition="Not Equal (Boolean)"(
     A=ShouldShow!(Minigame:Minigame Player Created Object Reference=@n6.Minigame, PropPossessorConfig:Prop Possessor Config Interface Interface=@n7) -> (Show:bool)
     B=IsVisible()
    ) -> (ReturnValue=~))
   n9 = Cast To Minigame_PlayerCreated!(execute=@n14.then, Object=Get CurrentMinigame() -> (CurrentMinigame:Fort Minigame Object Reference)) -> (then, CastFailed, AsMinigame Player Created:Minigame Player Created Object Reference) {is=title K2Node_DynamicCast}
   n10 = Branch(execute=[@n16.then_1, @n8.then], Condition=ShouldShow(Minigame=@n9, PropPossessorConfig=Get PropPossessorConfig()))
   > AnimateIn() {is=title K2Node_CallFunction}
   > SetVisibility(InVisibility=HitTestInvisible)
   > Bind Event to On Current Number Of Hiders Changed(self:Minigame Player Created Object Reference=@n9, Delegate=@n1) {is=title K2Node_AddDelegate}
   > Bind Event to On Starting Number Of Hiders Changed(self:Minigame Player Created Object Reference=@n9, Delegate=@n1) {is=title K2Node_AddDelegate}
   > UpdateCurrentNumberOfHiders!(
    execute
    self
    CurrentHiders:int=Get NumberOfHidersGameStart!(self:Minigame Player Created Object Reference=@n9) -> (NumberOfHidersGameStart:int=0)
   ) -> (then) {is=title K2Node_CallFunction}
   SetVisibility(execute=[@n10.else, @n9.CastFailed, @n14.else], InVisibility=Collapsed)
   > Branch(Condition=IsValid(Object=@n9))
   > Unbind Event from On Current Number Of Hiders Changed(self:Minigame Player Created Object Reference=@n9, Delegate=@n1) {is=title K2Node_RemoveDelegate}
   > Unbind Event from On Starting Number Of Hiders Changed(self:Minigame Player Created Object Reference=@n9, Delegate=@n1) {is=title K2Node_RemoveDelegate}
   Event Construct()
   > Parent: Construct()
   > SetVisibility(InVisibility=Collapsed)
   Event Destruct()
   > n11 = Cast To Minigame_PlayerCreated(Object=Get CurrentMinigame())
   > Unbind Event from On Prop Hunt Settings Changed(self:Minigame Player Created Object Reference=@n11, Delegate=@n6) {is=title K2Node_RemoveDelegate}
   n12 = Event OnDisplayMinigameInProgressUI!() -> (OutputDelegate, then, bDisplay:bool) {is=title K2Node_Event}
   > Parent: OnDisplayMinigameInProgressUI!(execute, bDisplay:bool=@n12.bDisplay) -> (then) {is=title K2Node_CallParentFunction}
   > n13 = Cast To Minigame_PlayerCreated(Object=Get CurrentMinigame())
   > Bind Event to On Prop Hunt Settings Changed(self:Minigame Player Created Object Reference=@n13, Delegate=@n6) {is=title K2Node_AddDelegate}
   > n14 = Branch(Condition=@n12.bDisplay)
   Event OnVisibilitySetEvent!() -> (OutputDelegate, then, InVisibility:ESlateVisibility Enum=Visible) {is=title K2Node_Event}
   n15 = Cast To PropPossessorConfigInterface!(execute=@n16.then_0, Object=Get PropHuntSettings!(self:Minigame Player Created Object Reference=@n9) -> (PropHuntSettings:Object Reference)) -> (then, CastFailed, AsProp Possessor Config Interface:Prop Possessor Config Interface Interface) {is=title K2Node_DynamicCast}
   > Set PropPossessorConfig(PropPossessorConfig:Prop Possessor Config Interface Interface=@n15) -> (Output_Get:Prop Possessor Config Interface Interface)
   n16 = Sequence(execute=@n9.then)
  ]}
 SetCurrentHidersText {nodes [
   n1 = SetCurrentHidersText!() -> (then, CurrentHiders:int) {is=title K2Node_FunctionEntry}
   > "SetText (Text)"(self=Get Remaining(), InText=Format Text!(Format="LOCTEXT(\"{0}\")", 0:int=@n1) -> (Result))
   > PlayAnimation(InAnimation=Get AddKill())
  ]}
 ShouldShow {nodes [
   n1 = ShouldShow!() -> (then, Minigame:Minigame Player Created Object Reference, PropPossessorConfig:Prop Possessor Config Interface Interface) {is=title K2Node_FunctionEntry}
   n2 = Get PropHuntSettings(self=@n1)
   > n3 = Branch(Condition=AND Boolean(A=IsValid(Object=@n2), B=IsPropHuntGame!(self:Minigame Player Created Object Reference=@n1) -> (IsPropHuntGame:bool)))
   > n4 = Get Prop Details!(execute, self:Prop Possessor Config Interface Interface=@n1.PropPossessorConfig) -> (then, Show Props Remaining:bool, Prop Health Scaling:bool, Allow Change of Prop:bool, Allow Cancel Prop Possession:bool) {is=title K2Node_CallFunction}
   > Return Node!(execute, Show:bool=@n4)
   Return Node(execute=@n3.else, Show=false)
  ]}
]
variables [
 ObjectivesComplete {type=number}
 PropPossessorConfig {default=None}
 TotalHiders {type=integer}
]
widgets {
 info {inheritedWidgetCount=1, parentClass=/Script/FortniteUI.MinigameWidgetBase, rootWidgetClass=&VerticalBox, widgetCount=5}
 widgets [
  Content {widgetClassPath=&VerticalBox, bIsVariable=true, widget=Content}
  Feedback {
   widgetClassPath=&HorizontalBox
   bIsVariable=true
   parent=Content
   widget=Feedback
   slot {class=VerticalBoxSlot, properties {horizontalAlignment=HAlign_Left}}
  }
  Message {
   widgetClassPath=&CommonTextBlock
   bIsVariable=true
   parent=Feedback
   widget=Message
   properties {
    text=Players Hiding:
    textTransformPolicy=ToUpper
    font {
     bForceMonospaced=false
     bMaterialIsStencil=false
     fontMaterial=null
     fontObject=&Burbank
     letterSpacing=0
     monospacedWidth=1
     size=20
     skewAmount=0
     typefaceFontName=Burbank-PlayerPower
     outlineSettings {
      bApplyOutlineToDropShadows=false
      bMiteredCorners=false
      bSeparateFillAlpha=false
      outlineMaterial=null
      outlineSize=3
      outlineColor=<rgba 0 0.21 0.31 1>
     }
    }
    shadowColorAndOpacity=<rgba 0 0 0 0.34>
    shadowOffset=<0 2>
   }
   slot {class=HorizontalBoxSlot, properties {horizontalAlignment=HAlign_Left, verticalAlignment=VAlign_Center}}
  }
  Remaining {
   widgetClassPath=&CommonTextBlock
   bIsVariable=true
   parent=Feedback
   widget=Remaining
   properties {
    text=0
    font {
     bForceMonospaced=false
     bMaterialIsStencil=false
     fontMaterial=null
     fontObject=&Burbank
     letterSpacing=0
     monospacedWidth=1
     size=20
     skewAmount=0
     typefaceFontName=Burbank-Black
     outlineSettings {
      bApplyOutlineToDropShadows=false
      bMiteredCorners=false
      bSeparateFillAlpha=false
      outlineMaterial=null
      outlineSize=3
      outlineColor=<rgba 0 0.2183 0.3125 1>
     }
    }
    shadowColorAndOpacity=<rgba 0.0052 0.0052 0.0056 0.8>
    shadowOffset=<3 3>
   }
   slot {class=HorizontalBoxSlot, properties {horizontalAlignment=HAlign_Left, verticalAlignment=VAlign_Center}}
  }
  TutorialHighlight {widgetClassPath=&TutorialHighlight}
 ]
}
