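# Prop Hunt minigame HUD widget that shows the countdown to the next prop ping sound.
# It listens to the Creative HUD context's countdown event, updating the number and playing a bounce or plain tick animation based on a warning threshold.
# Visibility depends on the prop possessor config's ping-timer setting and the owning controller type.
# A restarting failsafe timer collapses the widget when countdown updates stop.
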WidgetBlueprint /Game/Weapons/Prototype/ObjectMover/Blueprints/ObjectInteractionBehaviors/PossessProp/PropHunt_PropBeepTImer
animations [
 TickDown_Bounce {
  binding_count=1
  length_seconds=1.0018
  object=&self:TickDown_Bounce
  sequence {bindings [TimeText {
     id=#1
     possessed_class=CommonTextBlock
     tracks [
      {
       type=MovieScene2DTransformTrack
       sections [{
         type=MovieScene2DTransformSection
         end_frame=20
         start_frame=0
         channels [
          Translation.X {type=MovieSceneScriptingFloatChannel}
          Translation.Y {type=MovieSceneScriptingFloatChannel}
          Angle {type=MovieSceneScriptingFloatChannel}
          Scale.X {type=MovieSceneScriptingFloatChannel, keys [[0,1.5],[4.6,1],[17.0613,1],[20.036,0.8]]}
          Scale.Y {type=MovieSceneScriptingFloatChannel, keys [[0,1.5],[4.6,1],[17.0613,1],[20.036,0.8]]}
          Shear.X {type=MovieSceneScriptingFloatChannel}
          Shear.Y {type=MovieSceneScriptingFloatChannel}
         ]
        }]
      }
      {
       type=MovieSceneFloatTrack
       display_name=Render Opacity
       sections [{
         type=MovieSceneFloatSection
         end_frame=20
         start_frame=0
         channels [None {type=MovieSceneScriptingFloatChannel, keys [[0,1],[17.0613,1],[20.036,0]]}]
        }]
      }
     ]
    }]}
 }
 Slide {
  binding_count=1
  length_seconds=3
  object=&self:Slide
  sequence {bindings [Message {
     id=#2
     possessed_class=CommonTextBlock
     tracks [
      {
       type=MovieScene2DTransformTrack
       sections [{
         type=MovieScene2DTransformSection
         end_frame=60
         start_frame=0
         channels [
          Translation.X {type=MovieSceneScriptingFloatChannel, const [0,4]}
          Translation.Y {type=MovieSceneScriptingFloatChannel, keys [[0,0],[10,300],[50,300],[60,0]]}
          Angle {type=MovieSceneScriptingFloatChannel, const [0,1]}
          Scale.X {type=MovieSceneScriptingFloatChannel, keys [[0,1],[10,2],[50,2],[60,0]]}
          Scale.Y {type=MovieSceneScriptingFloatChannel, keys [[0,1],[10,2],[50,2],[60,0]]}
          Shear.X {type=MovieSceneScriptingFloatChannel, const [0,2]}
          Shear.Y {type=MovieSceneScriptingFloatChannel, const [0,1]}
         ]
        }]
      }
      {
       type=MovieSceneFloatTrack
       display_name=Render Opacity
       sections [{
         type=MovieSceneFloatSection
         end_frame=60
         start_frame=0
         channels [None {type=MovieSceneScriptingFloatChannel, keys [[0,0],[8,1],[53,1],[60,0]]}]
        }]
      }
     ]
    }]}
 }
 TickDown_Plain {
  binding_count=1
  length_seconds=1.0018
  object=&self:TickDown_Plain
  sequence {bindings [TimeText {
     id=#1
     possessed_class=CommonTextBlock
     tracks [
      {
       type=MovieScene2DTransformTrack
       sections [{
         type=MovieScene2DTransformSection
         end_frame=20
         start_frame=0
         channels [
          Translation.X {type=MovieSceneScriptingFloatChannel}
          Translation.Y {type=MovieSceneScriptingFloatChannel}
          Angle {type=MovieSceneScriptingFloatChannel}
          Scale.X {type=MovieSceneScriptingFloatChannel, const [1,1]}
          Scale.Y {type=MovieSceneScriptingFloatChannel, const [1,1]}
          Shear.X {type=MovieSceneScriptingFloatChannel}
          Shear.Y {type=MovieSceneScriptingFloatChannel}
         ]
        }]
      }
      {
       type=MovieSceneFloatTrack
       display_name=Render Opacity
       sections [{type=MovieSceneFloatSection, end_frame=20, start_frame=0, channels [None {type=MovieSceneScriptingFloatChannel, const [1,1]}]}]
      }
     ]
    }]}
 }
]
graphs [
 EventGraph {nodes [
   n1 = Set and Show Phase Timer!() -> (OutputDelegate, then, PhaseDuration:double=~, TimerTitle:Text, TimerFinishedMessage:Text) {is=title K2Node_CustomEvent}
   > n2 = Set PhaseDuration(PhaseDuration:double=@n1.PhaseDuration) -> (Output_Get:double)
   > Set CurrentTimeRemaining(CurrentTimeRemaining:double=@n2) -> (Output_Get:double)
   > Set TimerTitle(TimerTitle:Text=@n1.TimerTitle) -> (Output_Get:Text)
   > Set TimerFinishedMessage(TimerFinishedMessage:Text=@n1.TimerFinishedMessage) -> (Output_Get:Text)
   > SetPhaseTimerMessage!(execute, self, Timer Message:Text=Get TimerTitle() -> (TimerTitle:Text)) -> (then)
   > ShowTimer()
   StartTimer() {is=title K2Node_CustomEvent}
   > UpdateDisplayTime!(execute, self, TimeRemaining:double=0) -> (then)
   n3 = CountdownChanged!() -> (OutputDelegate, then, CountDownEventDisplayText:Text, TimeRemaining:float=~) {is=title K2Node_CustomEvent}
   > n4 = Set CurrentTimeRemaining(CurrentTimeRemaining=@n3.TimeRemaining)
   UpdateDisplayTime(execute=[@n19.then, @n4.then], TimeRemaining=@n4)
   > n5 = Branch(Condition="integer < integer"!(A:int=Round(A=@n4) {is=title K2Node_CallFunction}, B:int=0) -> (ReturnValue:bool) {is=title K2Node_CallFunction})
   > SetVisibility(InVisibility=Collapsed)
   ShowTimer(execute=@n5.else)
   > n6 = Clear and Invalidate Timer by Handle(Handle=Get FailsafeTimer() -> (FailsafeTimer:Timer Handle Structure))
   n7 = HideIfNoUpdates() {is=title K2Node_CustomEvent}
   n8 = Set Timer by Event(execute=@n6.then, Delegate=@n7, Time=1.5, bLooping=false)
   > Set FailsafeTimer(FailsafeTimer:Timer Handle Structure=@n8) -> (Output_Get:Timer Handle Structure)
   Event Construct()
   > SetVisibility(InVisibility=Collapsed)
   > Set PhaseDuration(PhaseDuration=5)
   > n9 = GetDynamicMaterial(self=Get Image_ProgressBar() -> (Image_ProgressBar:Image Object Reference))
   > Set ProgressBarDMI(ProgressBarDMI:Material Instance Dynamic Object Reference=@n9) -> (Output_Get:Material Instance Dynamic Object Reference)
   > SetColorAndOpacity(self:Image Object Reference=Get Image_ProgressBar(), InColorAndOpacity:Linear Color Structure=(R=0,G=0.5177,B=1,A=1))
   > SetPhaseTimerMessage(Timer Message=Get TimerTitle())
   n10 = Get CreativeHUDLContext!() -> (ReturnValue:Creative HUDLContext Object Reference) {is=title K2Node_GetBlueprintContextEx}
   > Is Valid(InputObject=@n10)
   > Bind Event to On Prop Play Sound Countdown Changed(self:Creative HUDLContext Object Reference=@n10, Delegate=@n3) {is=title K2Node_AddDelegate}
   > CountdownChanged!(execute, self, CountDownEventDisplayText:Text, TimeRemaining:float=-1) -> (then) {is=title K2Node_CallFunction}
   n11 = Event OnDisplayMinigameInProgressUI!() -> (OutputDelegate, then, bDisplay:bool) {is=title K2Node_Event}
   > Parent: OnDisplayMinigameInProgressUI!(execute, bDisplay:bool=@n11.bDisplay) -> (then) {is=title K2Node_CallParentFunction}
   > n12 = Branch(Condition=@n11.bDisplay)
   > n13 = Cast To Minigame_PlayerCreated!(execute, Object=Get CurrentMinigame() -> (CurrentMinigame:Fort Minigame Object Reference)) -> (then, CastFailed, AsMinigame Player Created:Minigame Player Created Object Reference) {is=title K2Node_DynamicCast}
   n14 = Get PropHuntSettings!(self:Minigame Player Created Object Reference=@n13) -> (PropHuntSettings:Object Reference)
   > n15 = Branch(Condition=OR Boolean(A=IsPropHuntGame!(self:Minigame Player Created Object Reference=@n13) -> (IsPropHuntGame:bool), B=IsValid(Object=@n14)))
   > n16 = Cast To PropPossessorConfigInterface!(execute, Object=@n14) -> (then, CastFailed, AsProp Possessor Config Interface:Prop Possessor Config Interface Interface) {is=title K2Node_DynamicCast}
   > n17 = Get Ping Details!(execute, self:Prop Possessor Config Interface Interface=@n16) -> (then, Show Ping Timer:bool, Ping Frequency:double) {is=title K2Node_CallFunction}
   > n18 = Branch(Condition=NOT Boolean(A=@n17))
   SetVisibility(execute=[@n18.then, @n15.else, @n13.CastFailed, @n12.else], InVisibility=Collapsed)
   n19 = Set TimerTitle()
   SetVisibility(execute=@n7.then, InVisibility=Collapsed)
   Event OnVisibilitySetEvent!() -> (OutputDelegate, then, InVisibility:ESlateVisibility Enum=Visible) {is=title K2Node_Event}
   Event Destruct()
  ]}
 GetPropHuntSettings {nodes [
   Inputs!() -> (In:Exec)
   > n1 = Cast To Minigame_PlayerCreated(Object=Get CurrentMinigame())
   > n2 = Cast To PropPossessorConfigInterface(Object=Get PropHuntSettings(self=@n1))
   > Outputs!(Valid:Exec, Invalid:Exec=[@n2.CastFailed, @n1.CastFailed], Possessor Config:Prop Possessor Config Interface Interface=@n2)
  ]}
 SetPhaseTimerMessage {nodes [
   n1 = SetPhaseTimerMessage!() -> (then, Timer Message:Text) {is=title K2Node_FunctionEntry}
   > "SetText (Text)"(self=Get Message() -> (Message:Common Text Object Reference), InText=@n1)
  ]}
 ShowTimer {nodes [
   ShowTimer() {is=title K2Node_FunctionEntry}
   n1 = GetOwningPlayer()
   > n2 = Cast To FortPlayerControllerAthena(Object=@n1)
   n3 = Branch(
    execute=[@n2.then, @n7.then]
    Condition="float > float"(A=Get CurrentTimeRemaining() -> (CurrentTimeRemaining:double), B=-4) {is=title K2Node_CallFunction}
   )
   > n4 = Get Prop Hunt Settings!(In:Exec) -> (Valid:Exec, Invalid:Exec, Possessor Config:Prop Possessor Config Interface Interface) {is=title K2Node_MacroInstance}
   > n5 = Get Ping Details(self=@n4)
   > n6 = Branch(Condition=@n5)
   SetVisibility(execute=[@n6.then, @n4.Invalid], InVisibility=SelfHitTestInvisible)
   > "SetText (Text)"(
    self=Get TimeText() -> (TimeText:Common Text Object Reference)
    InText="To Text (Float)"(Value=Get CurrentTimeRemaining(), MinimumIntegralDigits=1, MaximumIntegralDigits=324, MinimumFractionalDigits=0, MaximumFractionalDigits=0)
   )
   SetVisibility(execute=[@n6.else, @n3.else, @n7.CastFailed], InVisibility=Collapsed)
   n7 = Cast To FortLiveBroadcastController!(execute=@n2.CastFailed, Object=@n1) -> (then, CastFailed, AsFort Live Broadcast Controller:Fort Live Broadcast Controller Object Reference) {is=title K2Node_DynamicCast}
  ]}
 UpdateDisplayTime {nodes [
   SetColorAndOpacity(self=Get Image_ProgressBar(), InColorAndOpacity=(R=1,G=0.86,B=0,A=1))
   > PlaySound2D(Sound=Get Warning Sound() -> (Warning Sound:Sound Base Object Reference))
   n1 = UpdateDisplayTime!() -> (then, TimeRemaining:double=~) {is=title K2Node_FunctionEntry}
   > n2 = Set TimeBeforeBeep(TimeBeforeBeep:double=@n1) -> (Output_Get:double)
   > "SetText (Text)"(
    self=Get TimeText()
    InText="To Text (Float)"(Value=@n2, MinimumIntegralDigits=1, MaximumIntegralDigits=324, MinimumFractionalDigits=0, MaximumFractionalDigits=0)
   )
   > Branch(Condition="float >= float"(A=Get TimeBeforeBeep() -> (TimeBeforeBeep:double), B=1) {is=title K2Node_CallFunction})
   n3 = "float < float"!(A:double=Get TimeBeforeBeep(), B:double=Get WarningTime() -> (WarningTime:double)) -> (ReturnValue:bool) {is=title K2Node_CallFunction}
   > n4 = Branch(Condition=@n3)
   > PlayAnimation(InAnimation=Get TickDown_Bounce() -> (TickDown_Bounce:Widget Animation Object Reference))
   PlayAnimation(execute=@n4.else, InAnimation=Get TickDown_Plain() -> (TickDown_Plain:Widget Animation Object Reference))
   Branch(Condition="float < float"(A=Get TimeBeforeBeep(), B=Get ImminentTime() -> (ImminentTime:double)))
   > SetColorAndOpacity(self=Get Image_ProgressBar(), InColorAndOpacity=(R=1,G=0,B=0,A=1))
   > PlaySound2D(Sound=Get Imminent Sound() -> (Imminent Sound:Sound Base Object Reference))
   n5 = Branch(Condition="float >= float"(A=Get CurrentTimeRemaining(), B=-3))
   n6 = Create Event()
   > Set Timer by Event(Delegate=@n6, Time=1, bLooping=false)
   > SetPhaseTimerMessage(Timer Message=Get TimerFinishedMessage() -> (TimerFinishedMessage:Text)) {is=title K2Node_CallFunction}
   > SetColorAndOpacity(self=Get Image_ProgressBar(), InColorAndOpacity=(R=1,G=0,B=0,A=0))
   > Branch(Condition="Equal (Float)"!(A:double=Get CurrentTimeRemaining(), B:double=-1) -> (ReturnValue:bool) {is=title K2Node_CallFunction})
   > PlaySound2D(Sound=Get TimerFinishedStinger() -> (TimerFinishedStinger:Sound Base Object Reference))
   > PlayAnimation(InAnimation=Get Slide() -> (Slide:Widget Animation Object Reference))
   Set Timer by Event(execute=@n5.else, Delegate=@n6, Time=1, bLooping=false)
   Set Timer by Event(Delegate=@n6, Time=1, bLooping=false)
   AND Boolean(A=@n3, B="float >= float"(A=Get TimeBeforeBeep(), B=Get ImminentTime()))
  ]}
]
variables [
 CurrentTimeRemaining {type=number, default=1}
 FailsafeTimer {type=TimerHandle}
 ImminentTime {type=number, default=5}
 PhaseDuration {type=number, default=1}
 ProgressBarDMI {type=MaterialInstanceDynamic}
 TimerFinishedMessage {type=string, default="READY OR NOT HERE THEY COME!"}
 TimerTitle {type=string, default=Prop Ping in:}
 WarningTime {type=number, default=5}
]
widgets {
 info {inheritedWidgetCount=1, parentClass=/Script/FortniteUI.MinigameWidgetBase, rootWidgetClass=&HorizontalBox, widgetCount=6}
 widgets [
  Feedback {widgetClassPath=&HorizontalBox, bIsVariable=true, widget=Feedback}
  Message {
   widgetClassPath=&CommonTextBlock
   bIsVariable=true
   parent=Feedback
   widget=Message
   properties {
    text=Prop sound in:
    textTransformPolicy=ToUpper
    font {
     bForceMonospaced=false
     bMaterialIsStencil=false
     fontMaterial=null
     fontObject=&Burbank
     letterSpacing=0
     monospacedWidth=1
     size=20
     skewAmount=0
     typefaceFontName=Burbank-PlayerPower
     outlineSettings {
      bApplyOutlineToDropShadows=false
      bMiteredCorners=false
      bSeparateFillAlpha=false
      outlineMaterial=null
      outlineSize=3
      outlineColor=<rgba 0 0.21 0.31 1>
     }
    }
    shadowColorAndOpacity=<rgba 0 0 0 0.34>
    shadowOffset=<0 2>
   }
   slot {class=HorizontalBoxSlot, properties {verticalAlignment=VAlign_Center}}
  }
  Content {
   widgetClassPath=&Overlay
   bIsVariable=true
   parent=Feedback
   widget=Content
   slot {class=HorizontalBoxSlot, properties {horizontalAlignment=HAlign_Center, verticalAlignment=VAlign_Center}}
  }
  Image_ProgressBar {
   widgetClassPath=&Image
   bIsVariable=true
   parent=Content
   widget=Image_ProgressBar
   properties {brush {
     bIsDynamicallyLoaded=false
     drawAs=Image
     imageType=NoImage
     mirroring=NoMirror
     resourceName=None
     resourceObject=/Game/Creative/Devices/Siren/HollowCircularBackground_GamePhaseTimer
     tiling=NoTile
     imageSize=<70 70>
     margin=<ltrb 0 0 0 0>
     outlineSettings {
      bUseBrushTransparency=false
      roundingType=HalfHeightRadius
      width=0
      color {colorUseRule=UseColor_Specified, specifiedColor=<rgba 0 0 0 0>}
      cornerRadii=<0 0 0 0>
     }
     tintColor {colorUseRule=UseColor_Specified, specifiedColor=<rgba 1 1 1 1>}
     uVRegion {bIsValid=false, max=<0 0>, min=<0 0>}
    }}
   slot {class=OverlaySlot, properties {horizontalAlignment=HAlign_Center, verticalAlignment=VAlign_Center}}
  }
  TimeText {
   widgetClassPath=&CommonTextBlock
   bIsVariable=true
   parent=Content
   widget=TimeText
   properties {
    text=80
    font {
     bForceMonospaced=false
     bMaterialIsStencil=false
     fontMaterial=null
     fontObject=&Burbank
     letterSpacing=0
     monospacedWidth=1
     size=36
     skewAmount=0
     typefaceFontName=Burbank-Black
     outlineSettings {
      bApplyOutlineToDropShadows=false
      bMiteredCorners=false
      bSeparateFillAlpha=false
      outlineMaterial=null
      outlineSize=3
      outlineColor=<rgba 0 0.2183 0.3125 1>
     }
    }
    shadowColorAndOpacity=<rgba 0.0052 0.0052 0.0056 0.8>
    shadowOffset=<3 3>
   }
   slot {class=OverlaySlot, properties {horizontalAlignment=HAlign_Center, verticalAlignment=VAlign_Center, padding=<ltrb 0 0 -1.5 -3>}}
  }
  TutorialHighlight {widgetClassPath=&TutorialHighlight}
 ]
}
