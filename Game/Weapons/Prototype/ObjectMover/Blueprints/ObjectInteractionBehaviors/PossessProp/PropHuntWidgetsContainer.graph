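# PropHuntWidgetsContainer is a UMG widget that groups Prop Hunt HUD elements for the ObjectMover weapon's possess-prop feature.
# An Overlay root holds a horizontally centered HorizontalBox that places the props-remaining widget beside the prop beep timer widget, with padding between them.
# It is layout only: the event graph has unconnected default events and no variables or logic, so any behavior lives in the child widgets.
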WidgetBlueprint /Game/Weapons/Prototype/ObjectMover/Blueprints/ObjectInteractionBehaviors/PossessProp/PropHuntWidgetsContainer
graphs [EventGraph {nodes [
   Event PreConstruct()
   Event Construct()
   Event Tick()
  ]}]
variables []
widgets {
 info {inheritedWidgetCount=0, parentClass=&UserWidget, rootWidgetClass=&Overlay, widgetCount=4}
 widgets [
  Overlay_53 {widgetClassPath=&Overlay, widget=Overlay_53}
  HorizontalBox_24 {
   widgetClassPath=&HorizontalBox
   parent=Overlay_53
   widget=HorizontalBox_24
   slot {class=OverlaySlot, properties {horizontalAlignment=HAlign_Center, verticalAlignment=VAlign_Fill}}
  }
  PropHuntPropsRemaining {
   widgetClassPath=&/PropHuntPropsRemaining
   bIsVariable=true
   parent=HorizontalBox_24
   widget=PropHuntPropsRemaining
   slot {class=HorizontalBoxSlot, properties {horizontalAlignment=HAlign_Center, verticalAlignment=VAlign_Center}}
  }
  PropHunt_PropBeepTImer {
   widgetClassPath=&/PropHunt_PropBeepTImer
   bIsVariable=true
   parent=HorizontalBox_24
   widget=PropHunt_PropBeepTImer
   slot {class=HorizontalBoxSlot, properties {horizontalAlignment=HAlign_Center, verticalAlignment=VAlign_Center, padding=<ltrb 20 0 0 0>}}
  }
 ]
}
